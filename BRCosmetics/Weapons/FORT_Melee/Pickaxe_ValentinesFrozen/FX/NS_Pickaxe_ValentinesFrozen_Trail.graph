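# Niagara swing trail for the ValentinesFrozen pickaxe, combining an animation-driven ribbon with snow sprites.
# Main_Trail uses the FN anim-trail modules, has two ribbon renderers, computes a second color and double width, and drives material parameters over emitter looped age.
# SnowParticles spawn by distance moved around a sampled socket, inherit capped velocity, then drift under drag, downward acceleration, curl noise and point attraction.
# Time-of-day scalars dim both emitters' main color.
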NiagaraSystem /BRCosmetics/Weapons/FORT_Melee/Pickaxe_ValentinesFrozen/FX/NS_Pickaxe_ValentinesFrozen_Trail
dependencies {
 usedDataInterfaces [
  &NiagaraDataInterfaceVectorCurve
  &NiagaraDataInterfaceCurve
  &NiagaraDataInterfaceVector2DCurve
  &NiagaraDataInterfaceSocketReader
 ]
 usedDynamicInputs []
 usedModules [
  &self:SystemSpawnScript.SystemScriptSource.SystemScriptGraph.NiagaraNodeAssignment_0.SetVariables_#1
  &SystemState
 ]
}
emitters {
 Main_Trail {
  emitterSpawnScript [&FN_RibbonIDOffsetUtility()]
  emitterUpdateScript [
   &EmitterState()
   &FN_SpawnAnimTrail()
  ]
  particleSpawnScript [
   &FN_InitializeAnimTrailBoneIndecies()
   &FN_InitializeAnimTrail(Lifetime=0.3)
  ]
  particleUpdateScript [
   &SolveForcesAndVelocity(Clamp Velocity=false, Limit Acceleration=false)
   &FN_UpdateAnimTrail() {module=FN_UpdateAnimTrail001}
   &ScaleColor(
    Scale Alpha=&FloatFromCurve(FloatCurve=DataInterface() {propertyValues {Curve {keys [<at 0.0106 0 RCIM_Cubic>,<at 0.0995 0.74 RCIM_Cubic>,<at 1 0>]}}})
    Scale RGB=&VectorFromCurve(
     VectorCurve=DataInterface() {propertyValues {
       XCurve {keys [<at 0 0.0614>,<at 0.2 0.9062>,<at 0.4 0.0477>]}
       YCurve {keys [<at 0 0.1534>,<at 0.2 2.1974>,<at 0.4 0.2102>]}
       ZCurve {keys [<at 0 0.5>,<at 0.2 3>,<at 0.4 0.5>]}
      }}
     Scale Curve=&VectorFromFloat(Value=&FNTimeOfDayScalar(Night Scale=0.7, Morning Scale=0.9, Evening Scale=0.8))
    )
    ScaleA=true
    ScaleRGB=true
   )
   /SetVariables(Particles.Color 02:LinearColor=&MakeLinearColorFromVectorAndFloat(
     "Vector (RGB)"=&VectorFromCurve(VectorCurve=DataInterface() {propertyValues {
        XCurve {keys [<at 0 0.1227>,<at 0.5 4.8438>,<at 1 0.0955>]}
        YCurve {keys [<at 0 0.3069>,<at 0.5 11.1068>,<at 1 0.4203>]}
        ZCurve {keys [<at 0 1>,<at 0.5 15>,<at 1 1>]}
       }})
     "Float (Alpha)"=&FloatFromCurve(FloatCurve=DataInterface() {propertyValues {Curve {keys [<at 0.1252 0 RCIM_Cubic>,<at 0.3654 1 RCIM_Cubic>,<at 1 0>]}}})
    )) {isSetParameters=true}
   &ScaleRibbonWidth(Ribbon Width Scale=&FloatFromCurve(FloatCurve=DataInterface() {propertyValues {Curve {keys [{arriveTangent=-0.9322, interpMode=RCIM_Cubic, leaveTangent=-0.9322, tangentMode=RCTM_User, time=0, value=0.8994}, <at 1 0.1 RCIM_Cubic>]}}}))
   /SetVariables(Particles.RibbonWidthAlt:NiagaraFloat=&Multiply_Float(A=$Particles.RibbonWidth, B=2)) {isSetParameters=true}
   &DynamicMaterialParameters(
    Index 0 Param 1=&FloatFromCurve(
     FloatCurve=DataInterface() {propertyValues {Curve {keys [
         {arriveTangent=-23.6538, interpMode=RCIM_Cubic, leaveTangent=-23.6538, tangentMode=RCTM_User, time=0, value=4.0503}
         {arriveTangent=-4.507, interpMode=RCIM_Cubic, leaveTangent=-4.507, time=0.2499, value=1.1157}
         <at 0.8989 -0.0009 RCIM_Cubic>
        ]}}}
     CurveIndex=$Emitter.LoopedAge
    )
    Index 0 Param 2=1
    Index 0 Param 3=&FloatFromCurve(
     FloatCurve=DataInterface() {propertyValues {Curve {keys [
         {arriveTangent=-3.043, interpMode=RCIM_Cubic, leaveTangent=-3.043, tangentMode=RCTM_User, time=0.199, value=0.1267}
         <at 0.9991 -0.4903 RCIM_Cubic>
        ]}}}
     CurveIndex=$Emitter.LoopedAge
    )
    Index 0 Param 4=&FloatFromCurve(
     FloatCurve=DataInterface() {propertyValues {Curve {keys [
         <at 0.0986 0 RCIM_Cubic>
         {arriveTangent=0.4869, interpMode=RCIM_Cubic, leaveTangent=0.4869, time=0.1428, value=0.4623}
         {arriveTangent=-0.5406, interpMode=RCIM_Cubic, leaveTangent=-0.5406, time=0.4999, value=0.1954}
         <at 0.9981 0 RCIM_Cubic>
        ]}}}
     CurveIndex=$Emitter.LoopedAge
    )
    Module.Index0Float0Write=true
    Module.Index0Float1Write=true
    Module.Index0Float2Write=true
    Module.Index0Float3Write=true
   )
  ]
  rendererClasses [&NiagaraRibbonRendererProperties]
  renderers [
   {rendererClass=&NiagaraRibbonRendererProperties}
   {rendererClass=&NiagaraRibbonRendererProperties, rendererIndex=1}
  ]
 }
 SnowParticles {
  emitterUpdateScript [
   &EmitterState(Life Cycle Mode=::NewEnumerator1, Loop Behavior=::NewEnumerator1, Loop Delay=0.05, Loop Duration=0.3, UseLoopDelay=true)
   &SpawnPerUnit(
    Max Movement Threshold=2500
    Module.Use Max Movement Threshold=true
    Module.Use Movement Tolerance=false
    Spawn Spacing=25
    Use Spawn Probability=false
   )
  ]
  particleSpawnScript [
   &InitializeParticle(
    Lifetime Min=0.2
    Lifetime Mode=::NewEnumerator1
    Sprite Rotation Mode=::NewEnumerator2
    Sprite Size=&RandomRangeVector2D(Minimum=<2.5 2.5>, Maximum=<5 5>, Evaluation Type=::NewEnumerator1)
    Sprite Size Mode=::NewEnumerator4
    UsePositionOffset=false
    Write Material Random=true
   )
   &SocketLocation(
    Interpolate Socket Sampling=true
    SamplePosOffset=false
    Socket Index=1
    Socket Reader=$System.Socket Reader
    Socket Sampling Mode=::NewEnumerator3
    VelocitySpeedLimit=false
   )
   &ShapeLocation(
    MaskSpawnGroup=false
    Offset=<0 0 6>
    Offset Coordinate Space=::Local
    Offset Mode=::NewEnumerator0
    Override Seed=false
    Sphere Radius=5
    UseOwnerScale=true
   )
   &InheritVelocity(
    Inherited Velocity Speed Limit=20
    Speed Limit=true
    UseSpeedThreshold=false
    Velocity Source=$Particles.SkeletalMeshLocation.SampledVelocity
   )
  ]
  particleUpdateScript [
   &ParticleState()
   &SpriteRotationRate(Rotation Rate=&RandomRangeFloat(Minimum=-0.6, Maximum=0.6))
   &ScaleSpriteSize(
    Scale Factor=&Vector2DFromCurve(
     Vector2Curve=DataInterface() {propertyValues {
       XCurve {keys [
         <at -0.0029 0.0028 RCIM_Cubic>
         {arriveTangent=-0.0028, interpMode=RCIM_Cubic, leaveTangent=-0.0028, time=0.0805, value=0.9909}
         <at 1 0 RCIM_Cubic>
        ]}
       YCurve {keys [
         <at -0.0029 0.4648 RCIM_Cubic>
         {arriveTangent=-0.2417, interpMode=RCIM_Cubic, leaveTangent=-0.2417, time=0.0805, value=1}
         <at 1 0.2224 RCIM_Cubic>
        ]}
      }}
     ScaleCurve=<1 1>
    )
    Scale Sprite Size Mode=::NewEnumerator2
   )
   &Drag(Drag=2, Use Linear Drag=true, Use Rotational Drag=true)
   &AccelerationForce(Acceleration=<0 0 -100>)
   &CurlNoiseForce(Bias Noise=false, Noise Frequency=30, Noise Strength=200, Pan Noise=false)
   &PointAttractionForce(Attraction Radius=500, Attraction Strength=5, Kill Within Radius=false, Module.Use Falloff=true, UseAttractorOffset=false)
   &SolveForcesAndVelocity(Clamp Velocity=false, Limit Acceleration=false)
   &ScaleColor(
    Scale Alpha=&FloatFromCurve(FloatCurve=DataInterface() {propertyValues {Curve {keys [
         {arriveTangent=7.011, interpMode=RCIM_Cubic, leaveTangent=7.011, tangentMode=RCTM_User, time=0, value=0.0227}
         {arriveTangent=-0.0674, interpMode=RCIM_Cubic, leaveTangent=-0.0674, tangentMode=RCTM_User, time=1, value=2.9534}
        ]}}})
    Scale RGB=&VectorFromFloat(Value=&FNTimeOfDayScalar(Night Scale=0.7, Morning Scale=0.9, Evening Scale=0.8))
    ScaleA=true
    ScaleRGB=true
   )
  ]
  rendererClasses [&NiagaraSpriteRendererProperties]
  renderers [{rendererClass=&NiagaraSpriteRendererProperties}]
 }
}
summary {
 systemName=NS_Pickaxe_ValentinesFrozen_Trail
 emitters [
  {emitterName=Main_Trail, rendererClasses [&NiagaraRibbonRendererProperties]}
  {emitterName=SnowParticles, rendererClasses [&NiagaraSpriteRendererProperties]}
 ]
 userVariables [User.RibbonIDOffset {type=NiagaraInt32}]
}
